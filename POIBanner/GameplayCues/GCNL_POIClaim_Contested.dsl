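; Looping gameplay cue that plays audio while a POI Banner capture point is contested.
; Start stores the target, spawns the contested sound above the cue location and fades any lingering stop sound; removal fades the loop and spawns the stop sound.
; Both sounds receive an affiliation parameter distinguishing locally viewed, friendly/neutral or hostile instigators.
; A repeating timer feeds capture progress into the loop as ClaimPercent until removal clears it.
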(event EventOnLoopingStart(EitherVFX-preferthisversion)
  (params (MyTarget :type Actor%20Object%20Reference) (Parameters :type Gameplay%20Cue%20Parameters%20Structure%20%28by%20ref%29) (ParticleComponents :type Array%20of%20FXSystem%20Component%20Object%20References) (AudioComponents :type Array%20of%20Audio%20Component%20Object%20References))
  (Variables|Sound|SetCaptruePointTarget MyTarget)
  (CallFunction|PlayUncontestedSound Parameters MyTarget true))

(event EventOnRemoval(EitherVFX-preferthisversion)
  (params (MyTarget :type Actor%20Object%20Reference) (Parameters :type Gameplay%20Cue%20Parameters%20Structure%20%28by%20ref%29) (ParticleComponents :type Array%20of%20FXSystem%20Component%20Object%20References) (AudioComponents :type Array%20of%20Audio%20Component%20Object%20References) (BurstCameraShakeInstance :type Camera%20Shake%20Base%20Object%20Reference) (BurstDecalInstance :type Decal%20Actor%20Object%20Reference))
  (CallFunction|PlayUncontestedSound Parameters MyTarget))

(fn PlayerAffliation
  (params (Instigator :type Fort%20Player%20Pawn%20Object%20Reference))
  (bind _output_get (Variables|Sound|SetPlayerAffliationInt (Math|Integer|SelectInt 0 (Utilities|Select :Index (Team|GetTeamAffiliationforActortoLocalPlayer Instigator true true) :Friendly 1 :Neutral 1 :Hostile 2) (Pawn|IsLocallyViewed Instigator))))
  (Audio|Parameter|SetIntegerParameter (Variables|Components|GetContestedAudio) "PlayerAffliation" _output_get :!class /Script/Engine.AudioComponent))

(fn PlayUncontestedSound
  (params (Parameters :type Gameplay%20Cue%20Parameters%20Structure%20%28by%20ref%29) (Context :type Actor%20Object%20Reference) (Play :type Boolean))
  (bind (_normalizedmagnitude _rawmagnitude _effectcontext _matchedtagname _originaltag _aggregatedsourcetags _aggregatedtargettags _location _normal _instigator) (Ability|GameplayCue|BreakGameplayCueParameters Parameters))
  (bind _returnvalue (+ _location "0,0,100.000000"))
  (if Play
    (bind _asfort_player_pawn (Utilities|Casting|CastToFortPlayerPawn _instigator))
    (bind _returnvalue_2 (Audio|SpawnSoundatLocation (Variables|GetContext) (Variables|Sound|GetContestedSound) _returnvalue))
    (Variables|Components|SetContestedAudio _returnvalue_2)
    (CallFunction|PlayerAffliation _asfort_player_pawn)
    (bind _returnvalue_3 (Utilities|Time|SetTimerbyEvent (EventDispatchers|CreateEvent :!event UpdateClaimAudio) 0.1 true))
    (Variables|Sound|SetHandleClaimAudio _returnvalue_3)
    (bind _contestedstopaudio (Variables|Default|GetContestedStopAudio)
      (:then
        (Audio|Components|Audio|FadeOut _contestedstopaudio 0.25 0.0 "Linear" :!class /Script/Engine.AudioComponent))
      (:else))
    (else
      (bind _contestedaudio (Variables|Components|GetContestedAudio)
        (:then
          (Audio|Components|Audio|FadeOut _contestedaudio 0.25 0.0 "Linear" :!class /Script/Engine.AudioComponent)
          (bind _returnvalue_4 (Audio|SpawnSoundatLocation (Variables|GetContext) (Variables|Sound|GetContestedStopSound) _returnvalue))
          (bind _output_get (Variables|Default|SetContestedStopAudio _returnvalue_4))
          (Audio|Parameter|SetIntegerParameter _output_get "PlayerAffliation" (Variables|Sound|GetPlayerAffliationInt) :!class /Script/Engine.AudioComponent)
          (Utilities|Time|ClearandInvalidateTimerbyHandle (Variables|GetContext) (Variables|Sound|GetHandleClaimAudio)))
        (:else)))))

(fn UpdateClaimAudio
  (bind _captruepointtarget (Variables|Sound|GetCaptruePointTarget)
    (:then
      (bind _asbga_poibanner_capture_point (Utilities|Casting|CastToBGA_POIBanner_CapturePoint _captruepointtarget))
      (bind _contestedaudio (Variables|Components|GetContestedAudio)
        (:then
          (Audio|Parameter|SetFloatParameter _contestedaudio "ClaimPercent" (Variables|/POIBanner/Actors/CapturePoint/BGA_POIBanner_CapturePoint.BGA_POIBanner_CapturePoint_C|GetCurrentCaptureProgress _asbga_poibanner_capture_point) :!class /Script/AudioExtensions.AudioParameterControllerInterface))
        (:else)))
    (:else)))

(fn ConstructionScript)
