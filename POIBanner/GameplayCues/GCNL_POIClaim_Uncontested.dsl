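; Looping gameplay cue that plays positional audio while a POI banner capture point is claimed uncontested.
; On start, it compares the local player's team with the capturing team and spawns the loop above the point.
; A repeating timer feeds capture progress into the sound's ClaimPercent parameter.
; Removal clears the timer and fades into a stop sound; both sounds distinguish locally viewed instigator, capturing team and others via an affiliation parameter.
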(event EventOnLoopingStart(EitherVFX-preferthisversion)
  (params (MyTarget :type Actor%20Object%20Reference) (Parameters :type Gameplay%20Cue%20Parameters%20Structure%20%28by%20ref%29) (ParticleComponents :type Array%20of%20FXSystem%20Component%20Object%20References) (AudioComponents :type Array%20of%20Audio%20Component%20Object%20References))
  (bind (_normalizedmagnitude _rawmagnitude _effectcontext _matchedtagname _originaltag _aggregatedsourcetags _aggregatedtargettags _location _normal _instigator) (Ability|GameplayCue|BreakGameplayCueParameters Parameters))
  (Variables|Default|SetCaptruePointTarget MyTarget)
  (Variables|Default|SetPointLocation _location)
  (bind _asfort_player_pawn (Utilities|Casting|CastToFortPlayerPawn _instigator))
  (Variables|Default|SetPlayer%20Pawn _asfort_player_pawn)
  (CallFunction|UpdateAudio)
  (Variables|Default|GetUncontestedAudio
    (:then
      (bind _returnvalue (Utilities|Time|SetTimerbyEvent (EventDispatchers|CreateEvent :!event Custom|UpdateClaimAudio) 0.1 true))
      (Variables|Default|SetHandleClaimAudio _returnvalue))
    (:else)))

(event EventOnRemoval(EitherVFX-preferthisversion)
  (params (MyTarget :type Actor%20Object%20Reference) (Parameters :type Gameplay%20Cue%20Parameters%20Structure%20%28by%20ref%29) (ParticleComponents :type Array%20of%20FXSystem%20Component%20Object%20References) (AudioComponents :type Array%20of%20Audio%20Component%20Object%20References) (BurstCameraShakeInstance :type Camera%20Shake%20Base%20Object%20Reference) (BurstDecalInstance :type Decal%20Actor%20Object%20Reference))
  (CallFunction|PlayUncontestedSound false)
  (Utilities|Time|ClearandInvalidateTimerbyHandle MyTarget (Variables|Default|GetHandleClaimAudio)))

(event Custom|UpdateClaimAudio
  (bind _captruepointtarget (Variables|Default|GetCaptruePointTarget)
    (:then
      (bind _asbga_poibanner_capture_point (Utilities|Casting|CastToBGA_POIBanner_CapturePoint _captruepointtarget))
      (bind _uncontestedaudio (Variables|Default|GetUncontestedAudio)
        (:then
          (Audio|Parameter|SetFloatParameter _uncontestedaudio "ClaimPercent" (Variables|/POIBanner/Actors/CapturePoint/BGA_POIBanner_CapturePoint.BGA_POIBanner_CapturePoint_C|GetCurrentCaptureProgress _asbga_poibanner_capture_point) :!class /Script/AudioExtensions.AudioParameterControllerInterface))
        (:else)))
    (:else)))

(fn PlayerAffliation
  (params (AudioComponent :type Audio%20Component%20Object%20Reference))
  ;[1 1 1] Check to see if capturing team or not, then check for if that player is local for the capturing team.
  (Audio|Parameter|SetIntegerParameter AudioComponent "PlayerAffliation" (Math|Integer|SelectInt (Math|Integer|SelectInt 0 1 (Pawn|IsLocallyViewed (Variables|Default|GetPlayer%20Pawn))) 2 (Variables|Sound|GetIsCapturingTeam)) :!class /Script/Engine.AudioComponent))

(fn PlayUncontestedSound
  (params (Play :type Boolean))
  (bind _pointlocation (Variables|Default|GetPointLocation))
  (bind _returnvalue (+ _pointlocation "0,0,100.000000"))
  (if Play
    (bind _returnvalue_2 (Audio|SpawnSoundatLocation 0 (Variables|Sound|GetUncontestedSound) _returnvalue))
    (bind _output_get (Variables|Default|SetUncontestedAudio _returnvalue_2))
    (CallFunction|PlayerAffliation _output_get)
    (bind _uncontestedstopaudio (Variables|Default|GetUncontestedStopAudio)
      (:then
        (Audio|Components|Audio|FadeOut _uncontestedstopaudio 0.25 0.0 "Linear" :!class /Script/Engine.AudioComponent))
      (:else))
    (else
      (bind _uncontestedaudio (Variables|Default|GetUncontestedAudio)
        (:then
          (Audio|Components|Audio|FadeOut _uncontestedaudio 0.25 0.0 "Linear" :!class /Script/Engine.AudioComponent)
          (bind _returnvalue_3 (Audio|SpawnSoundatLocation 0 (Variables|Sound|GetUncontestedStopSound) _returnvalue))
          (bind _output_get_2 (Variables|Default|SetUncontestedStopAudio _returnvalue_3))
          (CallFunction|PlayerAffliation _output_get_2))
        (:else)))))

(fn UpdateAudio
  ;[1 1 1] Check local player and if player team is local capturing team or not.
  (bind _asfort_player_pawn (Utilities|Casting|CastToFortPlayerPawn (Pawn|GetControlledPawn (Fortnite|GetFirstLocalFortPlayerController (Variables|Default|GetCaptruePointTarget)))))
  (if (== (Variables|/POIBanner/Actors/CapturePoint/BGA_POIBanner_CapturePoint.BGA_POIBanner_CapturePoint_C|GetCapturingTeam (Utilities|Casting|CastToBGA_POIBanner_CapturePoint (Variables|Default|GetCaptruePointTarget))) (Teams|TeamtoIndex (Team|GetActorTeam _asfort_player_pawn) (Variables|Default|GetCaptruePointTarget)))
    (Variables|Sound|SetIsCapturingTeam true)
    (else
      (Variables|Sound|SetIsCapturingTeam false)))
  (CallFunction|PlayUncontestedSound true))

(fn ConstructionScript)
