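; Gameplay cue burst notify that plays audio when a POI banner capture completes.
; It casts the instigator to a Fort player pawn and plays a 2D stinger only when that pawn is friendly to the local player.
; It always spawns a capture sound slightly above the cue location.
; That sound's "PlayerAffliation" integer parameter distinguishes the locally viewed pawn, friendly or neutral players, and hostile players.
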(fn OnBurst(EitherVFX-preferthisversion)
  (params (MyTarget :type Actor%20Object%20Reference) (Parameters :type Gameplay%20Cue%20Parameters%20Structure%20%28by%20ref%29) (ParticleComponents :type Array%20of%20FXSystem%20Component%20Object%20References) (AudioComponents :type Array%20of%20Audio%20Component%20Object%20References) (BurstCameraShakeInstance :type Camera%20Shake%20Base%20Object%20Reference) (BurstDecalInstance :type Decal%20Actor%20Object%20Reference))
  (bind (_normalizedmagnitude _rawmagnitude _effectcontext _matchedtagname _originaltag _aggregatedsourcetags _aggregatedtargettags _location _normal _instigator) (Ability|GameplayCue|BreakGameplayCueParameters Parameters))
  ;[1 1 1] Play team 2D stinger sound for capture.
  (bind _asfort_player_pawn (Utilities|Casting|CastToFortPlayerPawn _instigator))
  (if (Utilities|Enum|Equal(Enum) (Team|GetTeamAffiliationforActortoLocalPlayer _asfort_player_pawn true true) "Friendly")
    (Audio|PlaySound2D MyTarget (Variables|Default|GetCapturedFriendlyStingerSound))
    (else
      _))
  (bind _returnvalue (Audio|SpawnSoundatLocation MyTarget (Variables|Default|GetCapturedSound) (+ _location "0,0,100.000000")))
  (Audio|Parameter|SetIntegerParameter _returnvalue "PlayerAffliation" (Math|Integer|SelectInt 0 (Utilities|Select :Index (Team|GetTeamAffiliationforActortoLocalPlayer _asfort_player_pawn true true) :Friendly 1 :Neutral 1 :Hostile 2) (Pawn|IsLocallyViewed _asfort_player_pawn)) :!class /Script/Engine.AudioComponent)
  (return))
